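# M_Torment_Decal is a masked Default Lit material in the Torment vehicle body cosmetics, named for that body's decal.
# It holds a decal base color texture sample and scalar parameters for Metallic, Specular and Roughness, though the file does not show how they are wired.
# It is flagged for skeletal mesh and mesh deformer use, with automatic usage setting and ray-traced shadow casting turned off.
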Material /VehicleCosmetics/Bodies/Torment/Materials/M_Torment_Decal
graph {
 properties {
  bAutomaticallySetUsageInEditor=false
  bCastRayTracedShadows=false
  bUsedWithMeshDeformer=true
  bUsedWithSkeletalMesh=true
  blendMode=BLEND_Masked
  usedShadingModels=Default Lit
 }
 nodes [
  ScalarParameter() {parameterName=Metallic}
  ScalarParameter() {defaultValue=0.5, parameterName=Specular}
  ScalarParameter() {defaultValue=0.25, parameterName=Roughness}
  TextureSample() {texture=/VehicleCosmetics/Bodies/Torment/Textures/Decal/T_Torment_Decal_BC}
 ]
}
